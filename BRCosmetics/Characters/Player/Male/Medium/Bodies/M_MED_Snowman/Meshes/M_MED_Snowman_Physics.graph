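# Physics asset for the Snowman medium male player body mesh, driving secondary motion for three bomb props hung from the character's pelvis.
# A pelvis capsule anchors three bomb capsules, each carrying a box body for its ring.
# Swing and twist limits tie each bomb to the pelvis, with bomb C held tighter than A and B.
# Each ring is constrained to its own bomb with very little twist, forming short two-link chains.
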PhysicsAsset /BRCosmetics/Characters/Player/Male/Medium/Bodies/M_MED_Snowman/Meshes/M_MED_Snowman_Physics
bodies {
 dyn_bombA [{shapeType=Capsule, length=2.8615, radius=4.3088, center=<4.2937 1.2722 -4.9893>, rotation=<rot -24.7231 -90 -29.6796>}]
 dyn_bombA_ring [{shapeType=Box, extentX=5.0304, extentY=1.0877, extentZ=5.1642, center=<1.5357 0.9145 -0.5586>, rotation=<rot 0.8047 -60.1903 -70.3023>}]
 dyn_bombB [{shapeType=Capsule, length=2.3615, radius=4.3256, center=<3.8867 0.0335 -4.5957>, rotation=<rot -5.8563 -90 -41.0927>}]
 dyn_bombB_ring [{shapeType=Box, extentX=5.51, extentY=1.01, extentZ=5.51, center=<2.1107 -0.1146 -0.6292>, rotation=<rot 0 -96.2143 -73.1901>}]
 dyn_bombC [{shapeType=Capsule, length=3.6115, radius=4.4254, center=<3.0443 -2.1507 -5.1043>, rotation=<rot 18.9789 -90 -36.7821>}]
 dyn_bombC_ring [{shapeType=Box, extentX=5.01, extentY=1.01, extentZ=5.01, center=<1.826 -1.1066 -0.6189>, rotation=<rot 0 -124.614 -72.9826>}]
 pelvis [{shapeType=Capsule, length=19.0046, radius=6.0359, center=<1.5388 -0.3905 -0.8152>, rotation=<rot 37.0485 98.8198 -90>}]
}
constraints [
 {bone1Name=dyn_bombA, bone2Name=pelvis, swing1LimitDegrees=45, swing2LimitDegrees=45, twistLimitDegrees=45}
 {bone1Name=dyn_bombB, bone2Name=pelvis, swing1LimitDegrees=45, swing2LimitDegrees=45, twistLimitDegrees=45}
 {bone1Name=dyn_bombC, bone2Name=pelvis, swing1LimitDegrees=25, swing2LimitDegrees=25, twistLimitDegrees=25}
 {bone1Name=dyn_bombA_ring, bone2Name=dyn_bombA, swing1LimitDegrees=45, swing2LimitDegrees=24.4286, twistLimitDegrees=3.4286}
 {bone1Name=dyn_bombB_ring, bone2Name=dyn_bombB, swing1LimitDegrees=45, swing2LimitDegrees=29.5714, twistLimitDegrees=2.1429}
 {bone1Name=dyn_bombC_ring, bone2Name=dyn_bombC, swing1LimitDegrees=45, swing2LimitDegrees=24.4286, twistLimitDegrees=1.7143}
]
