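# Animation Blueprint for the medium male Snowman body cosmetic, layering secondary motion onto a pose copied from the attached parent mesh, curves included.
# A Look At aims head_base at the head bone, then a RigidBody node simulates the Snowman physics asset relative to root with overridden gravity.
# Two Anim Dynamics chains for the bomb and bomb ring bones are unconnected to the output pose.
# The event graph has no logic.
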AnimBlueprint /BRCosmetics/Characters/Player/Male/Medium/Bodies/M_MED_Snowman/Meshes/M_MED_Snowman_AnimBP
target_skeleton=&/M_MED_Snowman_Skeleton
graphs [
 AnimGraph {nodes [
   n1 = Anim Dynamics - Bone: dyn_bombA(Alpha=1, ExternalForce="0,0,0") {
    angularDampingOverride=0.98
    bGravityOverrideInSimSpace=true
    gravityScale=0.8
    simulationSpace=RootRelative
    alphaBoolBlend {blendOption=Linear}
    boundBone {boneName=dyn_bombA}
    physicsBodyDefinitions [{
      sphereCollisionRadius=0
      boundBone {boneName=dyn_bombA}
      boxExtents=<30 30 30>
      constraintSetup {angularLimitsMax=<30 8.4 12>, angularLimitsMin=<-30 -30 -12>, angularTarget=<0 0 0>}
      localJointOffset=<0.23 0 -10>
     }]
   }
   n2 = Anim Dynamics - Bone: dyn_bombB(ComponentPose=@n1, Alpha=1, ExternalForce="0,0,0") {
    angularDampingOverride=0.98
    bGravityOverrideInSimSpace=true
    gravityScale=0.75
    simulationSpace=RootRelative
    alphaBoolBlend {blendOption=Linear}
    boundBone {boneName=dyn_bombB}
    physicsBodyDefinitions [{
      sphereCollisionRadius=0
      boundBone {boneName=dyn_bombB}
      boxExtents=<30 30 30>
      constraintSetup {angularLimitsMax=<30 8.4 12>, angularLimitsMin=<-30 -30 -12>, angularTarget=<0 0 0>}
      localJointOffset=<0.23 0 -15>
     }]
   }
   Anim Dynamics - Bone: dyn_bombC(ComponentPose=@n2, Alpha=1, ExternalForce="0,0,0") {
    angularDampingOverride=0.98
    bGravityOverrideInSimSpace=true
    gravityScale=0.825
    simulationSpace=RootRelative
    alphaBoolBlend {blendOption=Linear}
    boundBone {boneName=dyn_bombC}
    physicsBodyDefinitions [{
      sphereCollisionRadius=0
      boundBone {boneName=dyn_bombC}
      boxExtents=<30 30 30>
      constraintSetup {angularLimitsMax=<30 8.4 12>, angularLimitsMin=<-30 -30 -12>, angularTarget=<0 0 0>}
      localJointOffset=<0.23 0 -13>
     }]
   }
   n3 = Anim Dynamics - Bone: dyn_bombA_ring(Alpha=1, ExternalForce="0,0,0") {
    angularDampingOverride=0.98
    bGravityOverrideInSimSpace=true
    gravityScale=0.33
    simulationSpace=RootRelative
    alphaBoolBlend {blendOption=Linear}
    boundBone {boneName=dyn_bombA_ring}
    physicsBodyDefinitions [{
      sphereCollisionRadius=0
      boundBone {boneName=dyn_bombA_ring}
      boxExtents=<30 30 30>
      constraintSetup {angularLimitsMax=<3 0 15>, angularLimitsMin=<-3 -30 -15>, angularTarget=<0 0 0>}
      localJointOffset=<0.23 0 -10>
     }]
   }
   n4 = Anim Dynamics - Bone: dyn_bombB_ring(ComponentPose=@n3, Alpha=1, ExternalForce="0,0,0") {
    angularDampingOverride=0.98
    bGravityOverrideInSimSpace=true
    gravityScale=0.33
    simulationSpace=RootRelative
    alphaBoolBlend {blendOption=Linear}
    boundBone {boneName=dyn_bombB_ring}
    physicsBodyDefinitions [{
      sphereCollisionRadius=0
      boundBone {boneName=dyn_bombB_ring}
      boxExtents=<30 30 30>
      constraintSetup {angularLimitsMax=<3 0 15>, angularLimitsMin=<-3 -30 -15>, angularTarget=<0 0 0>}
      localJointOffset=<0.23 0 -10>
     }]
   }
   Anim Dynamics - Bone: dyn_bombC_ring(ComponentPose=@n4, Alpha=1, ExternalForce="0,0,0") {
    angularDampingOverride=0.98
    bGravityOverrideInSimSpace=true
    gravityScale=0.33
    simulationSpace=RootRelative
    alphaBoolBlend {blendOption=Linear}
    boundBone {boneName=dyn_bombC_ring}
    physicsBodyDefinitions [{
      sphereCollisionRadius=0
      boundBone {boneName=dyn_bombC_ring}
      boxExtents=<30 30 30>
      constraintSetup {angularLimitsMax=<3 0 15>, angularLimitsMin=<-3 -30 -15>, angularTarget=<0 0 0>}
      localJointOffset=<0.23 0 -10>
     }]
   }
   n5 = Look At - Bone: head_base(ComponentPose=Local To Component(LocalPose=Copy Pose From Mesh() {bCopyCurves=true, bUseAttachedParent=true}), Alpha=1) {
    alphaBoolBlend {blendOption=Linear}
    boneToModify {boneName=head_base}
    lookAtTarget {boneReference {boneName=head}}
    lookAt_Axis {axis=<1 0>}
   }
   n6 = RigidBody(ComponentPose=@n5) {
    bOverrideWorldGravity=true
    overridePhysicsAsset=&/M_MED_Snowman_Physics
    simulationSpace=BaseBoneSpace
    alphaBoolBlend {blendOption=Linear}
    baseBoneRef {boneName=root}
    componentAppliedLinearAccClamp=<5 5 5>
    componentLinearAccScale=<0.1 0.1 0.1>
    componentLinearVelScale=<5 5 5>
    overrideWorldGravity {z=-980}
   }
   Output Pose(Result=Component To Local(ComponentPose=@n6)) {layerGroup=None}
  ]}
 EventGraph {nodes [Event BlueprintUpdateAnimation()]}
]
variables []
