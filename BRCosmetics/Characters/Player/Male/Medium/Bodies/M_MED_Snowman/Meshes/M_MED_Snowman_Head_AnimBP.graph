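# Animation Blueprint for the head mesh of the medium male Snowman player cosmetic, adding subtle secondary motion to three leaf bones.
# It copies the pose from the attached parent mesh, then a Look At aims head_base at the head bone.
# Chained Anim Dynamics nodes jiggle the left, right and top leaves using damped springs, root-relative simulation and low blend alphas.
# All logic lives in the AnimGraph, with no variables or event logic.
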AnimBlueprint /BRCosmetics/Characters/Player/Male/Medium/Bodies/M_MED_Snowman/Meshes/M_MED_Snowman_Head_AnimBP
target_skeleton=&/M_MED_Snowman_Head_Skeleton
graphs [
 AnimGraph {nodes [
   n1 = Look At - Bone: head_base(ComponentPose=Local To Component(LocalPose=Copy Pose From Mesh() {bUseAttachedParent=true}), Alpha=1) {
    alphaBoolBlend {blendOption=Linear}
    boneToModify {boneName=head_base}
    lookAtTarget {boneReference {boneName=head}}
    lookAt_Axis {axis=<1 0>}
   }
   n2 = Anim Dynamics - Bone: dyn_leaf_lft(ComponentPose=@n1, Alpha=0.15, ExternalForce="0,0,0") {
    angularDampingOverride=0.985
    angularSpringConstant=255
    bAngularSpring=true
    bGravityOverrideInSimSpace=true
    bLinearSpring=true
    bOverrideAngularDamping=true
    bOverrideLinearDamping=true
    gravityScale=0.29
    linearDampingOverride=0.99
    linearSpringConstant=200
    simulationSpace=RootRelative
    alphaBoolBlend {blendOption=Linear}
    boundBone {boneName=dyn_leaf_lft}
    componentAppliedLinearAccClamp=<1000 1000 1000>
    componentLinearVelScale=<100 100 100>
    physicsBodyDefinitions [{
      sphereCollisionRadius=0
      boundBone {boneName=dyn_leaf_lft}
      boxExtents=<30 30 30>
      constraintSetup {
       angularTargetAxis=AxisY
       angularLimitsMax=<0 0 0>
       angularLimitsMin=<0 0 0>
       angularTarget=<0 0 0>
       linearAxesMax=<2 1.5 0.5>
       linearAxesMin=<-1.5 -1 -0.5>
      }
      localJointOffset=<0 0 0>
     }]
   }
   n3 = Anim Dynamics - Bone: dyn_leaf_rgt(ComponentPose=@n2, Alpha=0.18, ExternalForce="0,0,0") {
    angularDampingOverride=0.985
    angularSpringConstant=200
    bAngularSpring=true
    bGravityOverrideInSimSpace=true
    bLinearSpring=true
    bOverrideAngularDamping=true
    bOverrideLinearDamping=true
    gravityScale=0.31
    linearDampingOverride=0.99
    linearSpringConstant=150
    simulationSpace=RootRelative
    alphaBoolBlend {blendOption=Linear}
    boundBone {boneName=dyn_leaf_rgt}
    componentAppliedLinearAccClamp=<1000 1000 1000>
    componentLinearVelScale=<100 100 100>
    physicsBodyDefinitions [{
      sphereCollisionRadius=0
      boundBone {boneName=dyn_leaf_rgt}
      boxExtents=<20 10 15>
      constraintSetup {
       angularTargetAxis=AxisY
       angularLimitsMax=<0 0 0>
       angularLimitsMin=<0 0 0>
       angularTarget=<0 0 0>
       linearAxesMax=<1.5 1.5 0.5>
       linearAxesMin=<-1 -1 -0.5>
      }
      localJointOffset=<0 0 0>
     }]
   }
   n4 = Anim Dynamics - Bone: dyn_leaf_top(ComponentPose=@n3, Alpha=0.2, ExternalForce="0,0,0") {
    angularDampingOverride=0.985
    angularSpringConstant=255
    bAngularSpring=true
    bGravityOverrideInSimSpace=true
    bLinearSpring=true
    bOverrideAngularDamping=true
    bOverrideLinearDamping=true
    gravityScale=0.33
    linearDampingOverride=0.99
    linearSpringConstant=200
    simulationSpace=RootRelative
    alphaBoolBlend {blendOption=Linear}
    boundBone {boneName=dyn_leaf_top}
    componentAppliedLinearAccClamp=<1000 1000 1000>
    componentLinearVelScale=<100 100 100>
    physicsBodyDefinitions [{
      sphereCollisionRadius=0
      boundBone {boneName=dyn_leaf_top}
      boxExtents=<30 30 30>
      constraintSetup {
       angularTargetAxis=AxisY
       angularLimitsMax=<0 0 0>
       angularLimitsMin=<0 0 0>
       angularTarget=<0 0 0>
       linearAxesMax=<1.7728 2 1.2167>
       linearAxesMin=<-1.249 -0.29 -1.2302>
      }
      localJointOffset=<0 0 0>
     }]
   }
   Output Pose(Result=Component To Local(ComponentPose=@n4)) {layerGroup=None}
  ]}
 EventGraph {nodes [Event BlueprintUpdateAnimation()]}
]
variables []
